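# Master subsurface material for Hermes ice cliffs, layering landscape snow over ice and outputting matching Ice and Snow physical materials.
# An up-facing mask, widened on Nanite by vertex color and distance-field AO cracks, drives a height-lerped snow blend.
# Landscape RVT color and normals blend in by height and slope, with underwater snow tinted.
# Non-Nanite meshes add a fallback normal map; forward, mobile and proxy paths bypass RVT.
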Material /Hermes/Terrain/Cliffs/Materials/Ice/M_IceCliff_Hermes
graph {
 properties {
  bAutomaticallySetUsageInEditor=false
  bTangentSpaceNormal=false
  bUseMaterialAttributes=true
  bUsedWithInstancedStaticMeshes=true
  bUsedWithNanite=true
  bUsedWithSplineMeshes=true
  physMaterial=/Game/FortressPhysicalMaterials/Snow
  pixelDepthOffsetMode=PDOM_Legacy
  shadingModel=MSM_Subsurface
  usedShadingModels=Subsurface
 }
 nodes [
  n1 = TextureSampleParameter2D(UVs=StaticSwitchParameter(True=TextureCoordinate() {coordinateIndex=2}, False=TextureCoordinate()) {group=Ice Fallback, parameterName=Use UV2 for Fallback}) {
   desc="Use to support Non Nanite meshes fallbacks. "
   group=Ice Fallback
   parameterName=Fallback Normal
   samplerType=SAMPLERTYPE_Normal
   texture=&FlatNormalBlue
  }
  n2 = SphereMask(
   A=NaniteReplace!(Default=Transform(None=@n1.RGB), Nanite=VertexNormalWS())
   B=Constant3Vector() {constant=<rgba 0 0 1 1>}
   Radius=ScalarParameter() {defaultValue=0.9, group=2_Sides/Up, parameterName=Ice Up/Sides}
   Hardness=ScalarParameter() {defaultValue=0.35, group=2_Sides/Up, parameterName=Ice Up/Sides Contrast}
  )
  n3 = VertexColor()
  n4 = DistanceFieldApproxAO!(World Position, Normal, BaseDistance=Constant() {r=25}, Radius=Constant() {r=250}) {numSteps=3}
  n5 = Multiply(A=OneMinus(None=Multiply(A=@n3.G, B=@n4)), B=ScalarParameter() {parameterName=Snow In Cracks})
  FeatureLevelSwitch(Default=NaniteReplace(Default=@n2, Nanite=Max(A=@n2, B=@n5)), ES3_1=@n2)
  n6 = OneMinus(None=NamedRerouteUsage())
  n7 = ScalarParameter() {defaultValue=0.75, parameterName=RVT_BlendingSS_Mult}
  n8 = Saturate(None=Max(A=NamedRerouteUsage(), B=Multiply(A=@n6, B=@n7)))
  n9 = NamedRerouteUsage()
  n10 = NamedRerouteUsage()
  n11 = VertexNormalWS()
  n12 = RuntimeVirtualTextureSampleParameter!(UVs, World Position, Mip Bias=LinearInterpolate(Alpha=ComponentMask(None=@n11) {b=true}) {constA=6, constB=0}) {
   group=RVT
   materialType=BaseColor_Normal_Specular_YCoCg
   mipValueMode=RVTMVM_MipBias
   parameterName=RVT
   virtualTexture=/Game/Athena/Apollo/Maps/Landscape/VT/RVT_Apollo_Landscape
  }
  n13 = RuntimeVirtualTextureSampleParameter!(UVs, World Position) {
   group=RVT
   materialType=WorldHeight
   parameterName=RVTHeight
   virtualTexture=/Game/Athena/Apollo/Maps/Landscape/VT/RVT_Apollo_Landscape_Height
  }
  n14 = Subtract(A=Add(A=ScalarParameter() {group=RVT, parameterName=Offset RVT WP}, B=WorldPosition().Z), B=@n13.WorldHeight)
  n15 = NamedRerouteDeclaration(None=@n14) {name=RVT Height, nodeColor=<rgba 1 0 0.2344 1>}
  n16 = ScalarParameter() {
   defaultValue=25
   desc="Controls the blending of normals on sides of meshes, but not UP vector surfaces"
   group=RVT
   parameterName=Blend RVT Normals
  }
  n17 = ScalarParameter() {
   defaultValue=300
   desc="Used to control how far colors will blend with the rvt, as well sets blending distance of normals for Up looking surfaces"
   group=RVT
   parameterName=Blend RVT
  }
  n18 = Saturate(None=ComponentMask(None=VertexNormalWS()) {b=true})
  n19 = LinearInterpolate(A=@n16, B=@n17, Alpha=@n18)
  n20 = BlendMaterialAttributes(
   A=SetMaterialAttributes!(MaterialAttributes=@n10, Normal=@n12.Normal) {attributeSetTypes [#1]}
   B=@n10
   Alpha=Saturate(None=Divide(A=@n15, B=@n19))
  ) {vertexAttributeBlendType=UseA}
  n21 = SetMaterialAttributes!(MaterialAttributes=@n20, Base Color=@n12.BaseColor, Specular=@n12.BaseColor, Roughness=@n12.BaseColor, Subsurface Color) {attributeSetTypes [
    #2
    #3
    #4
    #5
   ]}
  n22 = BlendMaterialAttributes(A=@n21, B=@n20, Alpha=NamedRerouteUsage()) {vertexAttributeBlendType=UseA}
  n23 = ComponentMask(None=VertexNormalWS()) {b=true}
  n24 = OneMinus(None=@n23)
  n25 = Multiply(A=@n24, B=Abs(None=@n23))
  n26 = BlendMaterialAttributes(
   A=@n22
   B=@n9
   Alpha=StaticSwitchParameter(True=Saturate(None=@n24), False=Saturate(None=@n25)) {group=RVT, parameterName="Skip RVT on Bottom && Sides"}
  )
  n27 = StaticSwitchParameter(
   True=@n9
   False=BlendMaterialAttributes(A=@n9, B=StaticSwitchParameter(True=@n26, False=@n22) {group=RVT, parameterName=Skip RVT on Bottom}, Alpha=NamedRerouteUsage())
  ) {group=RVT, parameterName=SkipRVT}
  MaterialFunctionCall!(Mask=@n8, SlipperySnow_Mask, "Material;"=@n27) {materialFunction=/Game/Athena/Asteria/Environments/Landscape/Materials/MaterialFunctions/MF_SnowPostRVT}
  n28 = LinearInterpolate(A=@n16, B=@n17, Alpha=@n18)
  NamedRerouteDeclaration(None=MaterialFunctionCall(Transition Phase=Divide(A=@n14, B=@n28), Height Texture=NamedRerouteUsage()) {materialFunction=&HeightLerp}.Lerp Alpha No Contrast) {name=RVT Blending, nodeColor=<rgba 0.9922 1 0 1>}
  n29 = MaterialFunctionCall!(RVT=NamedRerouteUsage()) {materialFunction=/Game/Environments/Asteria/Rocks/Materials/MaterialFunctions/MF_Ice_2022}
  n30 = GetMaterialAttributes(None=MaterialFunctionCall!(Grass_UseCustomSlopeMask, Grass_CustomSlopeMask) {materialFunction=/BRCoreMaterials/Hermes/Terrain/Landscape/Materials/Layers/MF_Hermes_StaticMeshLandscapeLayerMatch}.SnowLayer) {attributeGetTypes [#1]}
  n31 = Transform(None=@n30.Normal) {transformSourceType=TRANSFORMSOURCE_World, transformType=TRANSFORM_Tangent}
  n32 = Constant()
  n33 = Constant() {r=1}
  n34 = NamedRerouteUsage()
  n35 = OneMinus(None=NamedRerouteDeclaration(None=@n29.Noise) {name=Noise, nodeColor=<rgba 0.5703 1 0 1>})
  n36 = GetMaterialAttributes(None=BlendMaterialAttributes(
    A=@n29.Result
    B=SetMaterialAttributes!(MaterialAttributes=@n30.MaterialAttributes, Normal=@n31, Subsurface Color=@n32, Opacity=@n33) {attributeSetTypes [#1, #5, #6]}
    Alpha=MaterialFunctionCall(Transition Phase=@n34, Height Texture=@n35) {materialFunction=&HeightLerp}.Lerp Alpha No Contrast
   )) {attributeGetTypes [#1]}
  n37 = TextureCoordinate() {coordinateIndex=2}
  n38 = TextureSampleParameter2D(UVs=@n37) {
   desc="Use to support Non Nanite meshes fallbacks. "
   group=Ice Fallback
   parameterName=Fallback Normal
   samplerSource=SSM_Wrap_WorldGroupSettings
   samplerType=SAMPLERTYPE_Normal
   texture=&FlatNormalBlue
  }
  n39 = Transform(None=MaterialFunctionCall!(UVs=@n37, Normal=@n38.RGB) {materialFunction=/Engine/Functions/Engine_MaterialFunctions01/Texturing/DeriveTangentBasis}.World Space Normal) {transformSourceType=TRANSFORMSOURCE_World, transformType=TRANSFORM_Tangent}
  n40 = TextureSampleParameter2D() {
   desc="Use to support Non Nanite meshes fallbacks. "
   group=Ice Fallback
   parameterName=Fallback Normal
   samplerSource=SSM_Wrap_WorldGroupSettings
   samplerType=SAMPLERTYPE_Normal
   texture=&FlatNormalBlue
  }
  n41 = StaticSwitchParameter(True=@n39, False=StaticSwitchParameter(True=@n39, False=@n40.RGB) {group=Ice Fallback, parameterName=Use UV2 for Fallback}) {group=Ice Fallback, parameterName=Use UV2 for Fallback}
  n42 = Constant3Vector() {constant=<rgba 1 1 0 0>}
  n43 = QualitySwitch(
   Default=Normalize(VectorInput=MaterialFunctionCall(BaseNormal=@n36.Normal, AdditionalNormal=@n41) {materialFunction=&BlendAngleCorrectedNormals}.Result)
   Low=Add(A=@n36.Normal, B=Multiply(A=@n41, B=@n42))
  )
  n44 = GetMaterialAttributes(None=SetMaterialAttributes!(MaterialAttributes=@n36.MaterialAttributes, Normal=NaniteReplace(Default=@n43, Nanite=@n36.Normal)) {attributeSetTypes [#1]}) {attributeGetTypes [#1]}
  NamedRerouteDeclaration(None=SetMaterialAttributes(MaterialAttributes=@n44.MaterialAttributes, Normal=Transform(None=@n44.Normal)) {attributeSetTypes [#1]}) {name=PreRVT, nodeColor=<rgba 0.4766 1 0 1>}
  n45 = WorldPosition()
  n46 = Saturate(None=ComponentMask(None=VertexNormalWS()) {b=true})
  n47 = Add(A=@n45.Z, B=LinearInterpolate(A=Constant() {r=155}, B=Constant() {r=-155}, Alpha=@n46))
  n48 = LinearInterpolate(A=Constant() {r=55}, B=Constant() {r=355}, Alpha=@n46)
  NamedRerouteDeclaration(None=Saturate(None=Divide(A=@n47, B=@n48))) {name=Ocean, nodeColor=<rgba 0.6406 1 0 1>}
  n49 = Add(A=Constant() {r=25}, B=@n45.Z)
  n50 = Constant() {r=155}
  NamedRerouteDeclaration(None=Saturate(None=Divide(A=@n49, B=@n50))) {name=Ocean_Line, nodeColor=<rgba 0.6406 1 0 1>}
  n51 = NamedRerouteDeclaration(None=Saturate(None=@n2)) {name=SnowMask, nodeColor=<rgba 0 0.2266 1 1>}
  MaterialExpressionLinearInterpolate_17 = LinearInterpolate(Alpha=@n51) {constA=1, constB=0}
  MaterialExpressionLinearInterpolate_16 = LinearInterpolate(Alpha=@n51)
  PhysicalMaterialOutput!(Ice=@MaterialExpressionLinearInterpolate_17, Snow=@MaterialExpressionLinearInterpolate_16) {inputs [
    {
     physicalMaterial=/Game/FortressPhysicalMaterials/Ice
     input {
      expression=&self:MaterialExpressionLinearInterpolate_17
      inputName=None
      mask=0
      maskA=0
      maskB=0
      maskG=0
      maskR=0
      outputIndex=0
     }
    }
    {
     physicalMaterial=&Snow
     input {
      expression=&self:MaterialExpressionLinearInterpolate_16
      inputName=None
      mask=0
      maskA=0
      maskB=0
      maskG=0
      maskR=0
      outputIndex=0
     }
    }
   ]}
  n52 = NamedRerouteUsage()
  n53 = VertexColor()
  n54 = BlendMaterialAttributes(
   A=@n27
   B=SetMaterialAttributes!(MaterialAttributes=@n27, Subsurface Color=Constant(), Opacity=Constant() {r=1}) {attributeSetTypes [#5, #6]}
   Alpha=StaticSwitchParameter(True=StaticSwitchParameter(True=@n53.R, False=@n8) {group=1_SnowMeshes, parameterName=Snow Meshes Vertex Mask, sortPriority=-1}, False=@n8) {group=1_SnowMeshes, parameterName=Snow Meshes, sortPriority=-1}
  )
  n55 = GetMaterialAttributes(None=@n54) {attributeGetTypes [#2, #4]}
  n56 = SetMaterialAttributes!(
   MaterialAttributes=@n55.MaterialAttributes
   Base Color=Multiply(A=@n55.Base Color, B=VectorParameter() {parameterName=UnderwaterSnowColor, defaultValue=<rgba 0.075 0.4154 0.75 1>}.RGB)
   Roughness=Constant() {r=1}
   Emissive Color=Constant()
  ) {attributeSetTypes [#2, #4, #7]}
  n57 = NamedRerouteUsage()
  n58 = MaterialFunctionCall(NoRVT=@n52, RVT=MaterialProxyReplace(Realtime=BlendMaterialAttributes(A=@n56, B=@n54, Alpha=@n57), MaterialProxy=@n52)) {materialFunction=&FN_RVT_Switch_Material}
  n59 = ShadingPathSwitch(Default=@n58.Result, Deferred=@n58.Result, Forward=@n52, Mobile=@n52)
  n60 = GetMaterialAttributes(None=@n59) {attributeGetTypes [#5, #2]}
  n61 = SetMaterialAttributes!(MaterialAttributes=@n60.MaterialAttributes, Base Color=Add(A=@n60.Subsurface Color, B=@n60.Subsurface Color)) {attributeSetTypes [#2]}
  SetMaterialAttributes(MaterialAttributes=GetMaterialAttributes(None=MaterialProxyReplace(Realtime=@n59, MaterialProxy=@n61)) {attributeGetTypes [#5, #2]}.MaterialAttributes) {attributeSetTypes [#2]}
  SetMaterialAttributes!(MaterialAttributes=@n30.MaterialAttributes, Normal=@n31) {attributeSetTypes [#1]}
 ]
}
